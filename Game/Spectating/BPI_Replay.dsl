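; BPI_Replay is a Blueprint interface in the spectating system for working with replay timeline markers.
; It declares FilterTimelineMarker, which takes a marker widget reference, and GetUserPlacedMarkers, which deals with an array of Fort Replay Timeline Marker objects.
; Both functions are signatures only, so filtering and marker collection logic lives outside this file.
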(fn FilterTimelineMarker
  (params (Marker :type Widget%20Object%20Reference)))

(fn GetUserPlacedMarkers
  (params (Markers :type Array%20of%20Fort%20Replay%20Timeline%20Marker%20Object%20References)))
