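; BP_BroadcastSpectatorPC is a broadcast spectator player controller Blueprint that configures squad markers for live broadcast viewing.
; Its construction script gets the marker component, sets the Athena SquadMarker HUD widget as its widget class, and sets its actor class from the FortLiveBroadcastController squad marker actor class variable.
; BeginPlay and Tick contain no logic.
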(event EventBeginPlay)

(event EventTick
  (params (DeltaSeconds :type Float%20%28single-precision%29)))

(fn ConstructionScript
  (bind _returnvalue (CallFunction|GetMarkerComponent))
  (MarkerComponent|SetMarkerWidgetClass _returnvalue "/Game/Athena/HUD/Team/SquadMarker.SquadMarker_C")
  (MarkerComponent|SetMarkerActorClass _returnvalue (Variables|FortLiveBroadcastController|GetSquadMarkerActorClass)))
