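; Spectator hover drone pawn that adds wind audio driven by flight altitude and speed.
; SetWindAudioEnabled spawns 2D altitude and movement wind loops, or fades them out when disabled.
; Update Wind Audio maps altitude and speed to clamped volumes, smoothed with FInterpTo, and applies them to the two loops.
; Tick, overlap, possession and AttachToBone events are empty, and nothing in this graph calls the audio update.
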(event EventBeginPlay)

(event Collision|EventActorBeginOverlap
  (params (OtherActor :type Actor%20Object%20Reference)))

(event EventTick
  (params (DeltaSeconds :type Float%20%28single-precision%29)))

(event EventPossessed
  (params (NewController :type Controller%20Object%20Reference)))

(event EventUnpossessed
  (params (OldController :type Controller%20Object%20Reference)))

(event Custom|Update%20Wind%20Audio
  (if (Variables|Audio|GetWindAudioEnabled)
    (bind _output_get (Variables|Audio|SetCurrentAltitudeWindVolume (Math|Interpolation|FInterpTo (Variables|Audio|GetCurrentAltitudeWindVolume) (CallFunction|GetWindVolumefromAltitude (HoverDrone|GetAltitude)) (Utilities|Time|GetWorldDeltaSeconds) 1e+01)))
    (Audio|Components|Audio|SetVolumeMultiplier (Variables|Default|GetAltitudeWindAudioComp) _output_get :!class /Script/Engine.AudioComponent)
    (bind _output_get_2 (Variables|Audio|SetCurrentMovementWindVolume (Math|Interpolation|FInterpTo (Variables|Audio|GetCurrentMovementWindVolume) (Math|Float|MapRangeClamped (Variables|Default|GetSpeed) 0.0 1.5e+03 0.0 1.0) (Utilities|Time|GetWorldDeltaSeconds) 1.0)))
    (Audio|Components|Audio|SetVolumeMultiplier (Variables|Default|GetMovementWindAudioComp) _output_get_2 :!class /Script/Engine.AudioComponent)))

(event Custom|AttachToBone
  (params (Bone :type Integer)))

(fn GetWindVolumeFromAltitude
  (params (Altitude :type Float%20%28double-precision%29))
  (return (Math|Float|MapRangeClamped Altitude (Variables|Audio|GetWindAudioAltMin) (Variables|Audio|GetWindAudioAltMax) (Variables|Audio|GetWindAudioVolMin) (Variables|Audio|GetWindAudioVolMax))))

(fn SetWindAudioEnabled
  (params (Enabled :type Boolean))
  (bind _output_get (Variables|Audio|SetWindAudioEnabled Enabled))
  (if _output_get
    (Variables|Default|GetAltitudeWindAudioComp
      (:then
        _)
      (:else
        (bind _returnvalue (Audio|SpawnSound2D (Variables|Audio|GetSoundWindAltitudeLoop)))
        (Variables|Default|SetAltitudeWindAudioComp _returnvalue)))
    (bind _returnvalue_2 (Audio|SpawnSound2D (Variables|Audio|GetSoundWindMovementLoop)))
    (Variables|Default|SetMovementWindAudioComp _returnvalue_2)
    (else
      (bind _altitudewindaudiocomp (Variables|Default|GetAltitudeWindAudioComp)
        (:then
          (Audio|Components|Audio|FadeOut _altitudewindaudiocomp 0.5 0.0 "Linear" :!class /Script/Engine.AudioComponent))
        (:else
          _))
      (bind _movementwindaudiocomp (Variables|Default|GetMovementWindAudioComp)
        (:then
          (Audio|Components|Audio|FadeOut _movementwindaudiocomp 0.5 0.0 "Linear" :!class /Script/Engine.AudioComponent))
        (:else)))))

(fn ConstructionScript)
