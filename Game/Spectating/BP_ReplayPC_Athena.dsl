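; Spectating player controller for Battle Royale (Athena) replays, mainly driving replay audio alongside marker and camera helpers.
; Replay context events and playback-rate changes drive pause and speed sound mixes, speed loops and HUD-sound gating; zoom events play camera loops.
; A one-second looping timer polls the view target to crossfade skydiving and parachute wind audio.
; It also hides the compass, sets marker classes, handles timeline markers and teleports the drone behind the followed player.
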(event Custom|OnReplayHudVisibilityChanged
  (params (IsVisible :type EHudVisibilityState%20Enum :default "FullyVisible"))
  ;[1 1 1] No HUD sounds if not visible
  (bind _output_get (Variables|Audio|SetbIsHudVisible (Utilities|Select :Index IsVisible :FullyVisible true :FullyHidden false :GameOnly false :ReplayOnly true)))
  (if (not _output_get)
    (bind _speeduploop (Variables|Default|GetSpeedupLoop)
      (:then
        (Audio|Components|Audio|FadeOut _speeduploop 1.0 0.0 "Linear" :!class /Script/Engine.AudioComponent))
      (:else
        _))
    (bind _slowdownloop (Variables|Default|GetSlowdownLoop)
      (:then
        (Audio|Components|Audio|FadeOut _slowdownloop 1.0 0.0 "Linear" :!class /Script/Engine.AudioComponent))
      (:else))))

(event Custom|On%20Pause%20State%20Changed
  (params (bPaused :type Boolean))
  (bind _pausemix (Variables|Default|GetPauseMIx))
  ;[1 1 1] Pause state change
  (if (Math|Boolean|NotEqual(Boolean) bPaused (Variables|Audio|GetbIsPaused))
    (bind _output_get (Variables|Audio|SetbIsPaused bPaused))
    (if _output_get
      (Audio|PushSoundMixModifier _pausemix)
      (CallFunction|UpdateSoundMix)
      (else
        (Audio|PopSoundMixModifier _pausemix)))))

(event EventBeginPlay
  ;[1 1 1] Only enable the UnderwaterAudioComponent if in a replay to prevent doubling up underwater sounds during gameplay
  (Variables|/Game/Sounds/Fort_Audio_Blueprints/UnderwaterAudioComponent/UnderwaterAudioComponent.UnderwaterAudioComponent_C|SetbEnabled (Replay|GetIsPlayingReplay) (Variables|Default|GetUnderwaterAudioComponent))
  (if (Replay|GetIsPlayingReplay)
    ; FORT-762010 - Makes it so the camera won't "push" away foliage when it's up close
    (Development|ExecuteConsoleCommand "SetGrassPushDownRadius 0")
    (Development|ExecuteConsoleCommand "HideHUDElement HUD.Athena.Compass") ; Hides the BR compass widget in Replay
    (else
      _))
  (ReplayContext|BindEventtoOnReplayHudVisibilityChanged (BlueprintContext|GetFortReplayContext) (EventDispatchers|CreateEvent :!event Custom|OnReplayHudVisibilityChanged))
  (ReplayContext|BindEventtoOnPauseStateChanged (BlueprintContext|GetFortReplayContext) (EventDispatchers|CreateEvent :!event Custom|On%20Pause%20State%20Changed))
  (Utilities|Time|SetTimerbyEvent (EventDispatchers|CreateEvent :!event Custom|Looping%20Audio%20Update) 1.0 true)
  (if (and (Pawn|IsLocalPlayerController) (Replay|GetIsPlayingReplay))
    (AddComponent|NODE_AddAmbientControllerComponent_Athena-0)))

(event Custom|Looping%20Audio%20Update
  (bind _pausemix (Variables|Default|GetPauseMIx))
  ;[1 1 1] Check if we should be playing loops, etc
  (CallFunction|HandleParachuteAudioState)
  ;[1 1 1] Update mix state when not paused.
  (Utilities|FlowControl|DoOnce false
    (:Completed
      (Utilities|IsValid _pausemix
        (:Is%20Valid
          (Audio|PopSoundMixModifier _pausemix))))))

(event Custom|Update%20Sound%20Mix
  (bind _pausemix (Variables|Default|GetPauseMIx))
  ;[1 1 1] Update mix state when not paused.
  (Utilities|FlowControl|DoOnce false
    (:Completed
      (Utilities|IsValid _pausemix
        (:Is%20Valid
          (Audio|PopSoundMixModifier _pausemix))))))

(event EventOnNotifyNewPlaybackMultiplier
  (params (NewMultiplier :type Float%20%28single-precision%29))
  (bind _playbackratemap (Variables|Audio|GetPlaybackRateMap))
  (bind _value (Utilities|Map|Find _playbackratemap NewMultiplier))
  ;[1 1 1] Lookup appropriate sound mix to activate for the new playback rate
  (if (!= NewMultiplier (Variables|Audio|GetCurrentPlayRate))
    (CallFunction|HandleSlowdown/SpeedupLoops NewMultiplier)
    (Audio|PushSoundMixModifier _value)
    (bind _values (Utilities|Map|Values (Variables|Audio|GetPlaybackRateMap)))
    (for _array_element _values
      (if (Utilities|NotEqual(Object) _array_element _value)
        (Audio|PopSoundMixModifier _array_element)))))

(event Custom|Handle%20Slowdown%20/%20Speedup%20Loops
  (params (Playback%20Multiplier :type Float%20%28double-precision%29))
  ;[1 1 1] Play speed up / slow down audio, and start/stop loops
  (if (Variables|Audio|GetbIsHudVisible)
    (Utilities|FlowControl|Sequence
      (:then_0
        (Audio|PlaySound2D (select (> Playback%20Multiplier (Variables|Audio|GetCurrentPlayRate)) "/Game/Sounds/Fort_UI_Sounds/Replay/DVR_Nav_Speedup_Cue.DVR_Nav_Speedup_Cue" "/Game/Sounds/Fort_UI_Sounds/Replay/DVR_Nav_Slowdown_Cue.DVR_Nav_Slowdown_Cue")))
      (:then_1
        (Variables|Audio|SetCurrentPlayRate Playback%20Multiplier)
        (if (> Playback%20Multiplier 1.0)
          (Utilities|FlowControl|Sequence
            (:then_0
              (bind _speeduploop (Variables|Default|GetSpeedupLoop)
                (:then
                  _)
                (:else
                  (bind _returnvalue (Audio|SpawnSound2D "/Game/Sounds/Fort_UI_Sounds/Replay/DVR_Nav_Fast_Loop_Cue.DVR_Nav_Fast_Loop_Cue"))
                  (Variables|Default|SetSpeedupLoop _returnvalue)))
              (Audio|Components|Audio|SetVolumeMultiplier _speeduploop (Math|Float|MapRangeClamped (Variables|Audio|GetCurrentPlayRate) 2.0 4.0 0.25 1.0) :!class /Script/Engine.AudioComponent))
            (:then_1
              (bind _slowdownloop (Variables|Default|GetSlowdownLoop)
                (:then
                  (Audio|Components|Audio|FadeOut _slowdownloop 1.0 0.0 "Linear" :!class /Script/Engine.AudioComponent))
                (:else))))
          (elif (< Playback%20Multiplier 1.0)
            (Utilities|FlowControl|Sequence
              (:then_0
                (bind _slowdownloop_2 (Variables|Default|GetSlowdownLoop)
                  (:then
                    _)
                  (:else
                    (bind _returnvalue_2 (Audio|SpawnSound2D "/Game/Sounds/Fort_UI_Sounds/Replay/DVR_Nav_Slow_Loop_Cue.DVR_Nav_Slow_Loop_Cue"))
                    (Variables|Default|SetSlowdownLoop _returnvalue_2)))
                (Audio|Components|Audio|SetVolumeMultiplier _slowdownloop_2 (- 1.0 (Math|Float|MapRangeClamped (Variables|Audio|GetCurrentPlayRate) 0.1 0.75 0.0 0.5)) :!class /Script/Engine.AudioComponent))
              (:then_1
                (bind _speeduploop_2 (Variables|Default|GetSpeedupLoop)
                  (:then
                    (Audio|Components|Audio|FadeOut _speeduploop_2 1.0 0.0 "Linear" :!class /Script/Engine.AudioComponent))
                  (:else))))
            (else
              (bind _speeduploop_3 (Variables|Default|GetSpeedupLoop)
                (:then
                  (Audio|Components|Audio|FadeOut _speeduploop_3 1.0 0.0 "Linear" :!class /Script/Engine.AudioComponent))
                (:else
                  _))
              (bind _slowdownloop_3 (Variables|Default|GetSlowdownLoop)
                (:then
                  (Audio|Components|Audio|FadeOut _slowdownloop_3 1.0 0.0 "Linear" :!class /Script/Engine.AudioComponent))
                (:else)))))))))

(event EventOnZoomOutBegin
  ;[1 1 1] Camera Zoom Sounds
  (bind _isgameplaycamera (CallFunction|IsGameplayCamera))
  (if (not _isgameplaycamera)
    (bind _returnvalue (Audio|SpawnSound2D "/Game/Sounds/Fort_UI_Sounds/Replay/DVR_Cam_ZoomOutLoop_Cue.DVR_Cam_ZoomOutLoop_Cue"))
    (Variables|Default|SetZoomOutLoop _returnvalue)))

(event EventOnZoomOutEnd
  ;[1 1 1] Camera Zoom Sounds
  (bind _isgameplaycamera (CallFunction|IsGameplayCamera))
  (if (not _isgameplaycamera)
    (Audio|Components|Audio|FadeOut (Variables|Default|GetZoomOutLoop) 0.2 0.0 "Linear" :!class /Script/Engine.AudioComponent)))

(event EventOnZoomInBegin
  ;[1 1 1] Camera Zoom Sounds
  (bind _isgameplaycamera (CallFunction|IsGameplayCamera))
  (if (not _isgameplaycamera)
    (bind _returnvalue (Audio|SpawnSound2D "/Game/Sounds/Fort_UI_Sounds/Replay/DVR_Cam_ZoomInLoop_Cue.DVR_Cam_ZoomInLoop_Cue"))
    (Variables|Default|SetZoomInLoop _returnvalue)))

(event EventOnZoomInEnd
  ;[1 1 1] Camera Zoom Sounds
  (bind _isgameplaycamera (CallFunction|IsGameplayCamera))
  (if (not _isgameplaycamera)
    (Audio|Components|Audio|FadeOut (Variables|Default|GetZoomInLoop) 0.2 0.0 "Linear" :!class /Script/Engine.AudioComponent)))

(event Custom|Handle%20Parachute%20Audio%20State
  ;[1 1 1] Handle Parachute audio if we detect our pawn is in the skydiving state
  (Utilities|FlowControl|Sequence
    (:then_0
      (bind _asfort_player_pawn_athena (Utilities|Casting|CastToFortPlayerPawnAthena (Pawn|GetViewTarget))
        (:then
          (Variables|Audio|SetbIsTargetSkydiving (and (Parachute|IsSkydiving _asfort_player_pawn_athena) (not (Parachute|IsParachuteOpen _asfort_player_pawn_athena))))
          (Variables|Audio|SetbIsTargetParachuting (Parachute|IsParachuteOpen _asfort_player_pawn_athena)))
        (:CastFailed
          (Variables|Audio|SetbIsTargetParachuting false)
          (Variables|Audio|SetbIsTargetSkydiving false)))
      (if (and (or (Utilities|Enum|Equal(Enum) (Variables|FortPlayerControllerSpectating|GetCurrentCameraType) "Gameplay") (Utilities|Enum|Equal(Enum) (Variables|FortPlayerControllerSpectating|GetCurrentCameraType) "ThirdPerson")) (or (Variables|Audio|GetbIsTargetParachuting) (Variables|Audio|GetbIsTargetSkydiving)))
        (if (not (Variables|Audio|GetbIsSkydivingAudioEnabled))
          (CallFunction|SetSkydivingAudioEnabled true))
        (else
          (if (Variables|Audio|GetbIsSkydivingAudioEnabled)
            (CallFunction|SetSkydivingAudioEnabled false)))))
    (:then_1
      (if (Variables|Audio|GetbIsSkydivingAudioEnabled)
        (if (Utilities|NotEqual(Object) (CallFunction|GetSkydivingSound) (Variables|Audio|GetCurrentSkydivingSound))
          (bind _skydivingaudioloop (Variables|Default|GetSkydivingAudioLoop)
            (:then
              (Audio|Components|Audio|FadeOut _skydivingaudioloop 1.0 0.0 "Linear" :!class /Script/Engine.AudioComponent))
            (:else
              _))
          (bind _output_get (Variables|Audio|SetCurrentSkydivingSound (CallFunction|GetSkydivingSound)))
          (bind _returnvalue (Audio|SpawnSound2D _output_get))
          (bind _output_get_2 (Variables|Default|SetSkydivingAudioLoop _returnvalue))
          (Audio|Components|Audio|FadeIn _output_get_2 1.0 1.0 0.0 "Linear" :!class /Script/Engine.AudioComponent))))))

(event Custom|TeleportToPawn
  ; Teleport to pawn
  (bind _ashover_drone_pawn (Utilities|Casting|CastToHoverDronePawn (Pawn|GetSpectatorPawn)))
  (Transformation|SetActorLocation _ashover_drone_pawn (+ (Actor|GetActorEyesViewPoint (PlayerState|GetPawn (CallFunction|GetFollowedPlayer))) (* (Transformation|GetActorForwardVector _ashover_drone_pawn) -2e+02)) false false))

(event EventFilterTimelineMarker
  (params (Marker :type Widget%20Object%20Reference))
  ; If marker isn't user placed, get rid of it
  (Widget|RemovefromParent Marker))

(event EventGetUserPlacedMarkers
  (params (Markers :type Array%20of%20Fort%20Replay%20Timeline%20Marker%20Object%20References))
  ; Assigns each marker a number that appears by hovering over it
  (for (_array_element _array_index) Markers
    (bind _astimeline_marker (Utilities|Casting|CastToTimelineMarker _array_element))
    (Class|TimelineMarker|UpdateMarkerCount _astimeline_marker (+ _array_index 1))
    (Widget|SetVisibility _astimeline_marker "SelfHitTestInvisible")))

(fn GetSkydivingSound
  (return (select (Variables|Audio|GetbIsTargetSkydiving) "/Game/Athena/Sounds/Wind_Falling/Glider_InAirAudio_Closed_1P_Cue.Glider_InAirAudio_Closed_1P_Cue" "/Game/Athena/Sounds/Wind_Falling/Glider_InAirAudio_Open_1P_Cue.Glider_InAirAudio_Open_1P_Cue")))

(fn IsGameplayCamera
  (return (Utilities|Enum|Equal(Enum) (Variables|FortPlayerControllerSpectating|GetCurrentCameraType) "Gameplay")))

(fn SetSkydivingAudioEnabled
  (params (Enabled :type Boolean))
  (bind _output_get (Variables|Audio|SetbIsSkydivingAudioEnabled Enabled))
  (if _output_get
    (bind _output_get_2 (Variables|Audio|SetCurrentSkydivingSound (CallFunction|GetSkydivingSound)))
    (bind _returnvalue (Audio|SpawnSound2D _output_get_2))
    (bind _output_get_3 (Variables|Default|SetSkydivingAudioLoop _returnvalue))
    (Audio|Components|Audio|FadeIn _output_get_3 1.0 1.0 0.0 "Linear" :!class /Script/Engine.AudioComponent)
    (else
      (bind _skydivingaudioloop (Variables|Default|GetSkydivingAudioLoop)
        (:then
          (Audio|Components|Audio|FadeOut _skydivingaudioloop 1.0 0.0 "Linear" :!class /Script/Engine.AudioComponent))
        (:else)))))

(fn ConstructionScript
  (bind _returnvalue (CallFunction|GetMarkerComponent))
  (MarkerComponent|SetMarkerWidgetClass _returnvalue "/Game/Athena/HUD/Team/SquadMarker.SquadMarker_C")
  (MarkerComponent|SetMarkerActorClass _returnvalue "/Game/Athena/HUD/Actors/B_AthenaPlayerMarker_WithCustomization.B_AthenaPlayerMarker_WithCustomization_C"))
