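; Spectator pawn that plays 2D wind audio matching the spectated Athena player's skydiving or open-parachute state.
; A looping 0.25-second timer polls the controller's view target instead of using Tick, setting the two state flags and stopping playing audio when the target changes.
; SetSkydivingAudioEnabled fades the loop in or out, while GetSkydivingSound picks the closed or open glider spectate cue, crossfading when that choice changes.
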(event EventBeginPlay
  (Utilities|Time|SetTimerbyEvent (EventDispatchers|CreateEvent :!event Custom|Handle%20Parachute%20Audio%20State) 0.25 true))

(event Collision|EventActorBeginOverlap
  (params (OtherActor :type Actor%20Object%20Reference)))

(event EventTick
  (params (DeltaSeconds :type Float%20%28single-precision%29)))

(event Custom|Handle%20Parachute%20Audio%20State
  (bind _skydivingaudioloop (Variables|Default|GetSkydivingAudioLoop))
  (Utilities|FlowControl|Sequence
    (:then_0
      (bind _asfort_player_pawn_athena (Utilities|Casting|CastToFortPlayerPawnAthena (Pawn|GetViewTarget (Pawn|GetController)))
        (:then
          (if (Utilities|NotEqual(Object) _asfort_player_pawn_athena (Variables|Default|GetCurrent%20Player%20Pawn))
            (Utilities|IsValid _skydivingaudioloop
              (:Is%20Valid
                (if (Audio|Components|Audio|IsPlaying _skydivingaudioloop)
                  (Audio|Components|Audio|Stop (Variables|Default|GetSkydivingAudioLoop))))
              (:Is%20Not%20Valid
                _)))
          (Variables|Default|SetCurrent%20Player%20Pawn _asfort_player_pawn_athena)
          (Variables|Default|SetbIsTargetSkydiving (and (Parachute|IsSkydiving _asfort_player_pawn_athena) (not (Parachute|IsParachuteOpen _asfort_player_pawn_athena))))
          (Variables|Default|SetbIsTargetParachuting (Parachute|IsParachuteOpen _asfort_player_pawn_athena)))
        (:CastFailed
          (Variables|Default|SetbIsTargetParachuting false)
          (Variables|Default|SetbIsTargetSkydiving false)))
      (if (or (Variables|Default|GetbIsTargetParachuting) (Variables|Default|GetbIsTargetSkydiving))
        (if (not (Variables|Default|GetbIsSkydivingAudioEnabled))
          (CallFunction|SetSkydivingAudioEnabled true))
        (else
          (if (Variables|Default|GetbIsSkydivingAudioEnabled)
            (CallFunction|SetSkydivingAudioEnabled false)))))
    (:then_1
      (if (Variables|Default|GetbIsSkydivingAudioEnabled)
        (if (Utilities|NotEqual(Object) (CallFunction|GetSkydivingSound) (Variables|Default|GetCurrentSkydivingSound))
          (bind _skydivingaudioloop_2 (Variables|Default|GetSkydivingAudioLoop)
            (:then
              (Audio|Components|Audio|FadeOut _skydivingaudioloop_2 1.5 0.0 "Linear" :!class /Script/Engine.AudioComponent))
            (:else
              _))
          (bind _output_get (Variables|Default|SetCurrentSkydivingSound (CallFunction|GetSkydivingSound)))
          (bind _returnvalue (Audio|SpawnSound2D _output_get))
          (bind _output_get_2 (Variables|Default|SetSkydivingAudioLoop _returnvalue))
          (Audio|Components|Audio|FadeIn _output_get_2 1.0 1.0 0.0 "Linear" :!class /Script/Engine.AudioComponent))))))

(fn GetSkydivingSound
  (return (select (Variables|Default|GetbIsTargetSkydiving) "/Game/Athena/Sounds/Wind_Falling/Glider_InAirAudio_Closed_1P_Spectate_Cue.Glider_InAirAudio_Closed_1P_Spectate_Cue" "/Game/Athena/Sounds/Wind_Falling/Glider_InAirAudio_Open_1P_Spectate_Cue.Glider_InAirAudio_Open_1P_Spectate_Cue")))

(fn SetSkydivingAudioEnabled
  (params (Enabled :type Boolean))
  (bind _output_get (Variables|Default|SetbIsSkydivingAudioEnabled Enabled))
  (if _output_get
    (bind _output_get_2 (Variables|Default|SetCurrentSkydivingSound (CallFunction|GetSkydivingSound)))
    (bind _returnvalue (Audio|SpawnSound2D _output_get_2))
    (bind _output_get_3 (Variables|Default|SetSkydivingAudioLoop _returnvalue))
    (Audio|Components|Audio|FadeIn _output_get_3 1.0 1.0 0.0 "Linear" :!class /Script/Engine.AudioComponent)
    (else
      (bind _skydivingaudioloop (Variables|Default|GetSkydivingAudioLoop)
        (:then
          (Audio|Components|Audio|FadeOut _skydivingaudioloop 1.0 0.0 "Linear" :!class /Script/Engine.AudioComponent))
        (:else)))))

(fn ConstructionScript)
