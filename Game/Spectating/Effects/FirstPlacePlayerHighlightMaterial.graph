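# Translucent, unlit, two-sided material that, by its name, highlights the first-place player for spectators, tinted by the PlayerTeamColor parameter.
# Emissive color comes from repeating bands that scroll along V, wobbled by a noise-driven sine wave.
# Opacity combines depth fade, a camera-distance ramp that dims it up close, and a V gradient peaking near one end.
# A capped Fresnel term sets the opacity ceiling.
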Material /Game/Spectating/Effects/FirstPlacePlayerHighlightMaterial
graph {
 properties {
  bUsedWithNanite=true
  blendMode=BLEND_Translucent
  pixelDepthOffsetMode=PDOM_Legacy
  shadingModel=MSM_Unlit
  twoSided=true
  usedShadingModels=Unlit
 }
 nodes [
  n1 = Constant() {r=64}
  n2 = Subtract(A=CameraPositionWS(), B=WorldPosition().XYZ)
  n3 = Clamp(None=SmoothStep(Value=MaterialFunctionCall(Vector 3=@n2) {materialFunction=/Engine/Functions/Engine_MaterialFunctions02/Utility/VectorLength}.V3 Length) {constMax=2500}) {minDefault=0.1}
  n4 = ComponentMask(None=TextureCoordinate()) {g=true}
  n5 = Constant() {r=0.95}
  n6 = Abs(None=Subtract(A=@n4, B=@n5) {constB=0.5})
  n7 = Divide(A=@n6, B=OneMinus(None=@n5))
  n8 = Power(Base=OneMinus(None=Divide(A=@n6, B=@n5))) {constExponent=3}
  n9 = If(A=@n4, B=@n5, "A > B"=Power(Base=OneMinus(None=@n7)), "A == B"=@n8, "A < B"=@n8)
  n10 = Constant() {r=1.5}
  Clamp(None=Multiply(A=DepthFade(FadeDistance=@n1), B=Multiply(A=@n3, B=@n9)) {constB=0.3}, Max=Min(A=Fresnel(ExponentIn=@n10)) {constB=0.65}) {maxDefault=0.3}
  n11 = Constant() {r=0.15}
  n12 = Constant()
  n13 = TextureCoordinate()
  n14 = TextureSample(UVs=@n13) {samplerType=SAMPLERTYPE_LinearColor, texture=&HighContrastNoise}
  n15 = ComponentMask(None=@n13) {r=true}
  n16 = Time()
  n17 = Multiply(A=Add(A=Add(A=@n14.G, B=@n15), B=Multiply(A=@n16) {constB=0.25})) {constB=3}
  n18 = Multiply(A=Multiply(A=@n14.R, B=Sine(None=@n17))) {constB=0.005}
  n19 = Add(A=ComponentMask(None=TextureCoordinate()) {g=true}, B=Multiply(A=Time() {period=400}) {constB=0.02})
  n20 = Fmod(A=Multiply(A=Add(A=@n18, B=@n19)) {constB=20}, B=Constant() {r=1})
  Multiply(
   A=VectorParameter() {parameterName=PlayerTeamColor, defaultValue=<rgba 1 0.25 0 0>}.RGB
   B=LinearInterpolate(Alpha=SmoothStep(Min=@n11, Max=@n12, Value=@n20)) {constA=1, constB=3}
  ) {constB=3}
 ]
}
