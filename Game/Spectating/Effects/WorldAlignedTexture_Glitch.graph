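# Material function in the spectating effects that projects a texture in world space with animated glitch distortion.
# World position divided by texture size yields three planar projections, blended by contrast-adjusted world normals into XYZ, XY and Z outputs.
# Projection UVs are multiplied by a glitch factor: summed panning sine-and-ceil gradient bands times panning noise, clamped by a time-driven sine minimum.
# A Vector2 input also pans the samples.
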MaterialFunction /Game/Spectating/Effects/WorldAlignedTexture_Glitch
graph {
 properties {bExposeToLibrary=true, description=Tiles a texture in worldspace, libraryCategoriesText [Misc,Texturing]}
 nodes [
  n1 = FunctionInput() {bUsePreviewValueAsDefault=true, description=Size of texture in worldspace units, inputName=TextureSize, previewValue=<64 64 64 6>}
  n2 = Divide(
   A=FunctionInput(Preview=WorldPosition() {worldPositionShaderOffset=WPT_ExcludeAllShaderOffsets}.XYZ) {bUsePreviewValueAsDefault=true, inputName=WorldPosition}
   B=Multiply(A=Abs(None=@n1)) {constB=-1}
  )
  n3 = ComponentMask(None=@n2) {b=true, r=true}
  n4 = MaterialFunctionCall(UV Channel=Panner() {speedX=3}) {materialFunction=&LinearGradient}
  n5 = Ceil(None=Sine(None=Sine(None=@n4.UGradient) {period=10}) {period=10})
  n6 = MaterialFunctionCall(UV Channel=Panner() {speedX=7}) {materialFunction=&LinearGradient}
  n7 = Ceil(None=Sine(None=Sine(None=@n6.UGradient) {period=10}) {period=10})
  n8 = TextureSample(UVs=Panner() {speedX=1, speedY=1}) {samplerType=SAMPLERTYPE_LinearGrayscale, texture=/Game/Global/Textures/Default/Noise/glitch_noise}
  n9 = Sine(None=Multiply(A=Time() {period=10}) {constB=5})
  n10 = ComponentMask(None=Multiply(A=Add(A=@n5, B=@n7), B=Clamp(None=@n8.RGB, Min=@n9))) {g=true, r=true}
  n11 = Time()
  n12 = FunctionInput(Preview=Constant2Vector()) {bUsePreviewValueAsDefault=true, inputType=FunctionInput_Vector2}
  n13 = FunctionInput(Preview=TextureObject() {texture=/Engine/EngineMaterials/DefaultDiffuse}) {inputName=TextureObject, inputType=FunctionInput_Texture2D}
  n14 = TextureSample(UVs=Panner(Coordinate=Multiply(A=@n3, B=@n10), Time=@n11, Speed=@n12), Tex=@n13) {samplerSource=SSM_Wrap_WorldGroupSettings}
  n15 = FunctionInput(Preview=StaticBool()) {bUsePreviewValueAsDefault=true, description="Default = False", inputName=Export Float 4, inputType=FunctionInput_StaticBool}
  n16 = ComponentMask(None=@n2) {b=true, g=true}
  n17 = TextureSample(UVs=Panner(Coordinate=Multiply(A=@n16, B=@n10), Time=@n11, Speed=@n12), Tex=@n13) {samplerSource=SSM_Wrap_WorldGroupSettings}
  n18 = FunctionInput(Preview=VertexNormalWS()) {bUsePreviewValueAsDefault=true, inputName=World Space Normal, previewValue=<0 0 1 1>}
  n19 = Abs(None=ComponentMask(None=@n18) {r=true})
  n20 = FunctionInput() {
   bUsePreviewValueAsDefault=true
   inputName=ProjectionTransitionContrast
   inputType=FunctionInput_Scalar
   sortPriority=10
   previewValue=<1 0 0 1>
  }
  n21 = LinearInterpolate(
   A=StaticSwitch(True=AppendVector(A=@n14.RGB, B=@n14.RGB), False=@n14.RGB, Value=@n15)
   B=StaticSwitch(True=AppendVector(A=@n17.RGB, B=@n17.RGB), False=@n17.RGB, Value=@n15)
   Alpha=ComponentMask(None=MaterialFunctionCall(In=@n19, Contrast=@n20) {materialFunction=&CheapContrast}.Result) {r=true}
  ) {desc=Float3}
  n22 = ComponentMask(None=@n2) {g=true, r=true}
  n23 = TextureSample(UVs=Panner(Coordinate=Multiply(A=@n22, B=@n10), Time=@n11, Speed=@n12), Tex=@n13) {samplerSource=SSM_Wrap_WorldGroupSettings}
  n24 = StaticSwitch(True=AppendVector(A=@n23.RGB, B=@n23.RGB), False=@n23.RGB, Value=@n15)
  n25 = MaterialFunctionCall(In=Abs(None=ComponentMask(None=@n18) {b=true}), Contrast=@n20) {materialFunction=&CheapContrast}
  FunctionOutput(None=LinearInterpolate(A=@n21, B=@n24, Alpha=ComponentMask(None=@n25.Result) {r=true})) {outputName=XYZ Texture, sortPriority=1}
  FunctionOutput(None=@n21) {outputName=XY Texture, sortPriority=1}
  FunctionOutput(None=@n24) {outputName=Z Texture, sortPriority=1}
 ]
}
