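# Translucent, unlit, two-sided material for the replay zone sphere in spectating, drawn as a hazy shell with animated grid and scanlines.
# Colour sums world-aligned glitch grid, noise-masked pulsing scanlines, depth-based tint and scene-intersection edge glow.
# Opacity combines depth, grid and edge terms, fading with height and near/far distance; an If yields a huge downward offset beyond a buffered far distance.
# Caustic and noise-grain branches end without visible consumers.
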Material /Game/Spectating/Effects/M_ReplayZoneSphere
graph {
 properties {
  bEnableResponsiveAA=true
  bUseTranslucencyVertexFog=false
  bUsedWithNanite=true
  blendMode=BLEND_Translucent
  pixelDepthOffsetMode=PDOM_Legacy
  shadingModel=MSM_Unlit
  twoSided=true
  usedShadingModels=Unlit
 }
 nodes [
  n1 = MaterialFunctionCall(
   Position=WorldPosition().XYZ
   Fade_Start_Distance=ScalarParameter() {defaultValue=4000, parameterName=Grid Fade Distance Start}
   Fade_End_Distance=ScalarParameter() {defaultValue=500, parameterName=Grid Fade Distance End}
  ) {materialFunction=&MF_Distance_SyncedFade_FX}
  n2 = MaterialFunctionCall!(
   TextureObject=TextureObject() {samplerType=SAMPLERTYPE_LinearColor, texture=/Game/Gadgets/Assets/Teleporter/Textures/T_Teleporter_GridSquares_01}
   TextureSize=ScalarParameter() {defaultValue=512, parameterName=GridWorldScale}
   WorldPosition
   Export Float 4=StaticBool()
   In
   World Space Normal
   ProjectionTransitionContrast=ScalarParameter() {defaultValue=13, parameterName=GridProjContrast}
  ) {materialFunction=&/WorldAlignedTexture_Glitch}
  n3 = ComponentMask(None=@n2.XYZ Texture) {r=true}
  n4 = Multiply(A=Time() {bOverride_Period=true, period=3.78}, B=ScalarParameter() {defaultValue=0.4, parameterName=Sine Speed Grid})
  n5 = Multiply(A=Sine(None=Add(A=@n3, B=@n4)) {period=0.5}, B=ComponentMask(None=@n2.XYZ Texture) {g=true})
  n6 = ScalarParameter() {defaultValue=5, parameterName=Grid Power}
  n7 = Multiply(
   A=Multiply(A=@n1.Result Inverted, B=Power(Base=@n5, Exp=@n6))
   B=VectorParameter() {parameterName=Grid Color, defaultValue=<rgba 0.15 0.15 0.15 0>}.RGB
  )
  n8 = Multiply(A=@n7, B=ScalarParameter() {defaultValue=0.3, parameterName=Glitch Opacity})
  n9 = TextureObject() {samplerType=SAMPLERTYPE_LinearColor, texture=/Game/Global/Textures/Default/Noise/T_Noise01}
  n10 = Constant3Vector() {constant=<rgba 512 512 2048 0>}
  n11 = StaticBool()
  n12 = Constant2Vector() {g=10, r=10}
  n13 = ScalarParameter() {defaultValue=1, parameterName=GridProjContrast_1}
  n14 = Time()
  n15 = Time()
  n16 = Add(A=Add(A=Sine(None=@n14) {period=0.5}, B=Sine(None=@n15)), B=Sine(None=Time()) {period=0.2})
  n17 = Add(
   A=WorldPosition().XYZ
   B=Multiply(A=Time() {bOverride_Period=true, period=5}, B=ScalarParameter() {defaultValue=300, parameterName=Line Pan Speed})
  )
  n18 = ScalarParameter() {defaultValue=0.003, parameterName=LineRepeatWorldScale}
  n19 = ComponentMask(None=Frac(None=Multiply(A=@n17, B=@n18))) {b=true}
  n20 = Add(A=Multiply(A=@n19, B=OneMinus(None=@n19)), B=ScalarParameter() {defaultValue=-0.2498, parameterName=Line Ceil Bias})
  n21 = Ceil(None=Multiply(A=Clamp(None=@n16), B=Saturate(None=@n20)))
  n22 = Multiply(
   A=Multiply(
    A=MaterialFunctionCall!(TextureObject=@n9, TextureSize=@n10, WorldPosition, Export Float 4=@n11, In=@n12, World Space Normal, ProjectionTransitionContrast=@n13) {materialFunction=/Game/Athena/SafeZone/WorldAlignedTexture_ScalePan}.XYZ Texture
    B=Multiply(A=@n21, B=@n1.Result Inverted) {desc=lines * distance fade}
   )
   B=VectorParameter() {parameterName=scanlines Color, defaultValue=<rgba 0.2483 0.2483 0.2483 1>}.RGB
  )
  Add(A=@n8, B=@n22)
  n23 = MaterialFunctionCall!(Distance Bias, Distance Divisor=ScalarParameter() {defaultValue=10000, parameterName=Saturation Divisor}, Scene Depth, Pixel Depth) {materialFunction=/Game/Effects/Fort_Effects/Materials/_Functions/MF_DistanceBehindTranslucency_01}
  n24 = LinearInterpolate(
   A=VectorParameter() {parameterName=Day Color Near, defaultValue=<rgba 0.0469 0.0469 0.0469 1>}.RGB
   B=VectorParameter() {parameterName=Day Color Far, defaultValue=<rgba 0.099 0.099 0.099 1>}.RGB
   Alpha=@n23.Normalized
  )
  n25 = Constant() {desc=Glitch strength, r=0.4}
  n26 = ScalarParameter() {defaultValue=250, parameterName=Intersection Edge Divisor}
  n27 = Multiply(
   A=Power(
    Base=MaterialFunctionCall(Distance Divisor=@n26) {materialFunction=&MF_DistanceBehindTranslucency_01}.Normalized Inverted
    Exp=ScalarParameter() {defaultValue=6, parameterName=Intersection Edge Exponent}
   )
   B=VectorParameter() {parameterName=Intersection Edge Color, defaultValue=<rgba 30 30 30 1>}.RGB
  )
  Add(A=Add(A=Add(A=@n22, B=@n24), B=Multiply(A=@n7, B=@n25)) {constB=0}, B=@n27) {desc=+ intersection edge}
  n28 = MaterialFunctionCall(
   TextureObject=TextureObject() {samplerType=SAMPLERTYPE_Normal, texture=/Game/TheOutpost/Textures/TilingWeakPointNormal}
   TextureSize=VectorParameter() {parameterName=CausticWorldScale, defaultValue=<rgba 29400 36750 21000 0>}.RGB
   Export Float 4=StaticBool()
   ProjectionTransitionContrast=ScalarParameter() {defaultValue=10, parameterName=CausticProjContrast}
  ) {materialFunction=&WorldAlignedTexture_ScalePan}
  n29 = Time() {bOverride_Period=true, desc=15.015014, period=15.015}
  n30 = ScalarParameter() {defaultValue=0.0666, desc=0.0666, parameterName=Sine Speed A}
  n31 = Add(A=@n28.XYZ Texture, B=ConstantBiasScale(None=Multiply(A=@n29, B=@n30)) {bias=-0.5, scale=3.3})
  n32 = Sine(None=Divide(A=WorldPosition().XYZ, B=VectorParameter() {parameterName=Variation Scale, defaultValue=<rgba 30000 30000 30000 0>}.RGB))
  n33 = Time() {bOverride_Period=true, period=20.202}
  n34 = ScalarParameter() {defaultValue=0.0499, desc=0.04995, parameterName=Sine Speed B}
  n35 = Add(A=ConstantBiasScale(None=Multiply(A=@n33, B=@n34)) {bias=-0.5, scale=4.4}, B=@n28.XYZ Texture)
  n36 = Add(A=Sine(None=Add(A=@n31, B=@n32)) {desc=3.3, period=3.3}, B=Sine(None=Add(A=@n35, B=@n32)) {desc=4.4, period=4.4})
  n37 = OneMinus(None=Abs(None=Desaturation(None=@n36)))
  n38 = Saturate(None=@n37)
  n39 = Round(None=@n38)
  n40 = ScalarParameter() {defaultValue=0.25, parameterName=Caustic Round LERP}
  Add(A=Multiply(A=VectorParameter() {parameterName=Caustics Color, defaultValue=<rgba 0.4 0.4 0.4 1>}.RGB, B=LinearInterpolate(A=@n38, B=@n39, Alpha=@n40))) {desc=+ Grid Blocks}
  n41 = Sine(None=MaterialFunctionCall(UV Channel=Panner() {speedX=1}) {materialFunction=&LinearGradient}.UGradient)
  n42 = Power(Base=Multiply(A=@n41, B=@n41)) {constExponent=5000}
  n43 = Sine(None=Time())
  n44 = Panner() {speedY=1}
  n45 = Panner(Coordinate=Multiply(
    A=TextureCoordinate()
    B=TextureSample(UVs=@n44) {samplerType=SAMPLERTYPE_LinearColor, texture=/Game/Environments/Foliage/Textures/T_NoiseMask02_M}.R
   )) {speedX=10, speedY=10}
  n46 = TextureSample(UVs=@n45) {samplerType=SAMPLERTYPE_LinearColor, texture=&T_NoiseMask02_M}
  n47 = TextureSample(UVs=@n45) {texture=/Game/Effects/Fort_Effects/Textures/Fire/T_FireNoise_02}
  n48 = Time()
  n49 = Multiply(A=Power(Base=Subtract(A=@n46.R, B=@n47.R), Exp=ScalarParameter() {defaultValue=5, parameterName=Noise power}), B=Ceil(None=Sine(None=@n48)))
  Add(A=Multiply(A=Multiply(A=@n42), B=Frac(None=@n43)), B=@n49)
  n50 = TextureSample(UVs=Panner(Coordinate=TextureCoordinate() {uTiling=3, vTiling=0.1}) {speedY=-1}) {texture=/Game/Global/Textures/Default/Noise/Noise}
  n51 = ScalarParameter() {defaultValue=10, parameterName=Grain power}
  Clamp(None=Add(A=@n49, B=Power(Base=@n50.RGB, Exp=@n51)))
  n52 = WorldPosition()
  n53 = ScalarParameter() {defaultValue=16000, parameterName=HeightGradientScale}
  n54 = Saturate(None=ComponentMask(None=Divide(A=@n52.XYZ, B=@n53)) {b=true})
  n55 = Add(
   A=@n8
   B=LinearInterpolate(
    A=ScalarParameter() {defaultValue=0.4, parameterName=Base Opacity Near}
    B=ScalarParameter() {defaultValue=0.7, parameterName=Base Opacity Far}
    Alpha=@n23.Normalized
   )
  )
  n56 = ScalarParameter() {defaultValue=45000, parameterName=Start Fade Distance Far}
  n57 = MaterialFunctionCall(
   Opacity In=Multiply(A=OneMinus(None=@n54), B=Add(A=@n27, B=@n55))
   Position=WorldPosition().XYZ
   Fade_Start_Distance=@n56
   Fade_End_Distance=ScalarParameter() {defaultValue=100, parameterName=End Fade Distance Near}
  ) {materialFunction=&MF_Distance_SyncedFade_FX}
  n58 = Constant()
  If(
   A=@n57.Distance
   B=Multiply(A=@n56, B=ScalarParameter() {defaultValue=1.55, parameterName=Buffer}) {constB=1.36, desc=add some buffer}
   "A > B"=Constant3Vector() {constant=<rgba 0 0 -999999 0>}
   "A == B"=@n58
   "A < B"=@n58
  )
  MaterialFunctionCall(
   Opacity In=@n57.Result Inverted
   Position=WorldPosition().XYZ
   Fade_Start_Distance=ScalarParameter() {defaultValue=400, parameterName=Start Fade Distance Near}
   Fade_End_Distance=ScalarParameter() {defaultValue=100, parameterName=End Fade Distance Near}
  ) {materialFunction=&MF_Distance_SyncedFade_FX}
  Multiply(A=@n37)
  ScalarParameter() {defaultValue=2048, parameterName=GridWorldScale_1}
 ]
}
