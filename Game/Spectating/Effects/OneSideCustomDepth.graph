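# OneSideCustomDepth is a masked Default Lit material in the spectating effects, apparently meant for one-sided custom depth, as its name suggests.
# It compares the dot product of pixel normal and camera vector to a threshold, giving 0 for camera-facing pixels and 1 for grazing or away-facing ones.
# This view-angle cutoff presumably drives the opacity mask, alongside a flat dark grey colour constant.
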Material /Game/Spectating/Effects/OneSideCustomDepth
graph {
 properties {bUsedWithNanite=true, blendMode=BLEND_Masked, pixelDepthOffsetMode=PDOM_Legacy, usedShadingModels=Default Lit}
 nodes [
  Constant3Vector() {constant=<rgba 0.2 0.2 0.2 0>}
  n1 = Constant() {r=1}
  If(A=DotProduct(A=PixelNormalWS(), B=CameraVectorWS()), B=Constant() {r=0.4}, "A > B"=Constant(), "A == B"=@n1, "A < B"=@n1)
 ]
}
