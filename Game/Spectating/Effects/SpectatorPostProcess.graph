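# Spectator post-process material that outlines custom-depth objects in team or selection colours and desaturates parts of relevancy zones.
# Edges come from four custom-depth taps with resolution-scaled offsets, coloured by a code-set colour or stencil-indexed team palettes, scaled by time of day.
# Stencil and depth tests show team outlines when occluded and selection outlines when visible, excluding vehicles.
# Relevancy-zone stencil pixels blend toward greyscale using a depth mask controlled by InsideRelevancy.
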Material /Game/Spectating/Effects/SpectatorPostProcess
graph {
 properties {
  bEnableResponsiveAA=true
  blendMode=BLEND_Translucent
  blendableLocation=BL_SceneColorBeforeDOF
  blendablePriority=-5
  materialDomain=MD_PostProcess
  pixelDepthOffsetMode=PDOM_Legacy
  shadingModel=MSM_Unlit
  usedShadingModels=Unlit
 }
 nodes [
  ConstantBiasScale(None=Sine(None=Time() {bIgnorePause=true}))
  n1 = NamedRerouteUsage()
  n2 = CollectionParameter() {
   collection=&FortniteMaterialParameters
   parameterName=FortniteActiveTimeOfDay
  }
  n3 = SceneTexture() {sceneTextureId=PPI_CustomStencil}
  n4 = VectorParameter() {desc=code controled, group=Outline, parameterName=OutlineColor, defaultValue=<rgba 1 1 1 0>}
  n5 = ScalarParameter() {group=Outlines, parameterName=UseCreativeOutlines, sliderMax=1}
  n6 = Constant()
  n7 = Subtract(
   A=ComponentMask(None=SceneTexture() {sceneTextureId=PPI_CustomStencil}.Color) {r=true}
   B=CollectionParameter() {collection=/Game/Packages/Fortress_SharedMaterials/Stencil_ID_Definitions/Stencil_ID_Definitions, parameterName=RedOverlayID}
  ) {constB=11}
  n8 = Constant2Vector() {r=0.125}
  n9 = TextureSample(UVs=Subtract(A=Multiply(A=@n7, B=@n8) {constB=4}, B=Constant2Vector() {r=0.0625}) {constB=11}) {texture=/Game/Athena/HUD/Spectator/Art/E32019/T_TeamColors_Creative}
  n10 = Constant2Vector() {r=0.02}
  n11 = TextureSample(UVs=Subtract(A=Multiply(A=@n7, B=@n10) {constB=4}, B=Constant2Vector() {r=0.01}) {constB=11}) {texture=/Game/Characters/CharacterColorSwatches/TeamIdentification/OutlineColoursTexture}
  n12 = Multiply(
   A=DotProduct(A=@n2, B=Constant4Vector() {constant=<rgba 1 1 1 0.4>}.RGBA)
   B=If(
    A=ComponentMask(None=@n3.Color) {r=true}
    B=CollectionParameter() {collection=&Stencil_ID_Definitions, parameterName=RedOverlayID}
    "A > B"=If(A=@n5, B=@n6, "A > B"=@n9.RGB, "A == B"=@n11.RGB, "A < B"=@n11.RGB)
    "A == B"=@n4.RGB
    "A < B"=@n4.RGB
   )
  )
  n13 = ComponentMask(None=ScreenPosition().ViewportUV) {g=true, r=true}
  n14 = Subtract(A=Divide(A=ViewSize(), B=Constant() {r=320}), B=Constant() {r=2}) {constB=0}
  n15 = SceneTexelSize()
  n16 = Multiply(A=Clamp(None=Ceil(None=@n14)) {maxDefault=2, minDefault=1}, B=@n15) {constA=2}
  n17 = Constant()
  n18 = AppendVector(A=ComponentMask(None=@n16) {desc=x, r=true}, B=@n17)
  n19 = Add(A=@n13, B=@n18) {desc=x+}
  n20 = AppendVector(A=@n17, B=ComponentMask(None=@n16) {desc=y, g=true}) {desc=y+}
  n21 = Add(A=@n13, B=@n20) {desc=y+}
  n22 = Constant() {r=-1}
  n23 = Add(A=@n13, B=Multiply(A=@n18, B=@n22)) {desc=x-}
  n24 = Add(A=@n13, B=Multiply(A=@n22, B=@n20)) {desc=y-}
  n25 = ComponentMask(None=MaterialFunctionCall!(
    Base=SceneTexture() {sceneTextureId=PPI_CustomDepth}.Color
    BaseSampleBias
    Scale
    Sample1=SceneTexture(UVs=@n19) {sceneTextureId=PPI_CustomDepth}.Color
    Sample2=SceneTexture(UVs=@n21) {sceneTextureId=PPI_CustomDepth}.Color
    Sample3=SceneTexture(UVs=@n23) {sceneTextureId=PPI_CustomDepth}.Color
    Sample4=SceneTexture(UVs=@n24) {sceneTextureId=PPI_CustomDepth}.Color
    Divisor=Constant() {r=-16}
    UseCenteredLines
    UseRGB
   ) {materialFunction=/Game/Packages/jl_tests/combine4Samples}.Result) {r=true}
  n26 = MaterialFunctionCall!(ID=CollectionParameter() {collection=&Stencil_ID_Definitions, parameterName=Vehicles}) {materialFunction=/Game/Packages/Fortress_SharedMaterials/Stencil_ID_Definitions/StencilMaskCompare}
  n27 = MaterialFunctionCall(ID=CollectionParameter() {collection=&Stencil_ID_Definitions, parameterName=Vehicles_Lockon}) {materialFunction=&StencilMaskCompare}
  n28 = OneMinus(None=Saturate(None=Add(A=@n26.Result, B=@n27.Result)))
  n29 = ComponentMask(None=SceneTexture() {sceneTextureId=PPI_CustomStencil}.Color) {r=true}
  n30 = CollectionParameter() {collection=&Stencil_ID_Definitions, parameterName=RedOverlayID}
  n31 = Constant() {r=1}
  n32 = Constant()
  n33 = SceneTexture() {sceneTextureId=PPI_CustomDepth}
  n34 = ScalarParameter() {defaultValue=-2, parameterName=Custom Depth Offset Amount}
  n35 = ComponentMask(None=Subtract(A=Add(A=@n33.Color, B=@n34), B=SceneTexture() {sceneTextureId=PPI_SceneDepth}.Color)) {r=true}
  n36 = Ceil(None=Clamp(None=Multiply(A=@n35) {constB=-1, desc="If player outline, invert occlusion check"})) {desc=z tested objects black z tested - everything else white}
  n37 = CollectionParameter() {
   collection=&Stencil_ID_Definitions
   parameterName=SelectionOutlineID_CodeReferenced
  }
  n38 = Ceil(None=Clamp(None=@n35)) {desc=z tested objects black z tested - everything else white}
  n39 = Add(
   A=Multiply(A=If(A=@n29, B=@n30, "A > B"=@n31, "A == B"=@n32, "A < B"=@n32), B=OneMinus(None=@n36))
   B=Multiply(
    A=MaterialFunctionCall(ID=@n37) {materialFunction=&StencilMaskCompare}.Result
    B=OneMinus(None=@n38)
   )
  )
  n40 = NamedRerouteUsage()
  n41 = Subtract(
   A=Add(A=SceneTexture() {sceneTextureId=PPI_CustomDepth}.Color, B=ScalarParameter() {defaultValue=-2, parameterName=Custom Depth Offset Amount_1})
   B=SceneTexture() {sceneTextureId=PPI_SceneDepth}.Color
  )
  n42 = ScalarParameter() {defaultValue=1, parameterName=InsideRelevancy, sliderMax=1, sliderMin=-1}
  n43 = Clamp(None=Multiply(A=ComponentMask(None=@n41) {r=true}, B=@n42) {desc="If player outline, invert occlusion check"})
  n44 = LinearInterpolate(
   A=@n40
   B=Desaturation(None=@n40, Fraction=Constant() {r=1})
   Alpha=OneMinus(None=Ceil(None=@n43) {desc=z tested objects black z tested - everything else white}) {desc=occlusion mask}
  )
  n45 = LinearInterpolate(
   A=LinearInterpolate(
    A=NamedRerouteUsage()
    B=LinearInterpolate(A=@n1, B=@n12, Alpha=@n25) {desc=outline z tested objects when opaque}
    Alpha=Multiply(A=@n28, B=@n39)
   )
   B=@n44
   Alpha=MaterialFunctionCall(ID=CollectionParameter() {collection=&Stencil_ID_Definitions, parameterName=RelevancyZoneID}) {materialFunction=&StencilMaskCompare}.Result
  )
  If(A=@n42, B=Constant(), "A > B"=@n45, "A == B"=@n45, "A < B"=@n44)
  Multiply(A=@n15) {constB=4}
  NamedRerouteDeclaration(None=ComponentMask(None=SceneTexture() {sceneTextureId=PPI_PostProcessInput0}.Color) {b=true, g=true, r=true}) {name=SceneColor.RGB, nodeColor=<rgba 0.4297 0 1 1>}
  ScalarParameter() {defaultValue=3, parameterName=NumberOfPixels}
  n46 = Subtract(
   A=ComponentMask(None=SceneTexture() {sceneTextureId=PPI_CustomStencil}.Color) {r=true}
   B=CollectionParameter() {collection=&Stencil_ID_Definitions, parameterName=RedOverlayID}
  ) {constB=11}
  n47 = Constant2Vector() {r=0.02}
  TextureSample(UVs=Subtract(A=Multiply(A=@n46, B=@n47) {constB=4}, B=Constant2Vector() {r=0.01}) {constB=11}) {texture=&OutlineColoursTexture}
 ]
}
