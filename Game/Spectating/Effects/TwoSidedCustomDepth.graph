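# TwoSidedCustomDepth is a minimal two-sided material among the spectating effects, named for use with custom depth rendering.
# It uses Default Lit shading, is flagged for Nanite, and its only expression is a constant dark grey color.
# The material itself holds no custom depth logic, textures or parameters.
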Material /Game/Spectating/Effects/TwoSidedCustomDepth
graph {
 properties {bUsedWithNanite=true, pixelDepthOffsetMode=PDOM_Legacy, twoSided=true, usedShadingModels=Default Lit}
 nodes [Constant3Vector() {constant=<rgba 0.2 0.2 0.2 0>}]
}
