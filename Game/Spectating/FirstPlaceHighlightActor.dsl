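; First-place player highlight actor that marks the leading player for spectators with a team-colored mesh.
; The construction script creates a dynamic material instance on the static mesh, and team color changes update its color parameter.
; Changes to spectator camera type, highlight enablement or followed-by-camera state re-evaluate the mesh's visibility through the highlighter's ShouldbeVisible check.
; On BeginPlay the root uses absolute rotation, so it ignores its parent's orientation.
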(event EventBeginPlay
  (Transformation|SetAbsolute (Variables|Transformation|GetRootComponent) false true false))

(event EventSpectatorCameraTypeChanged
  (params (NewCameraType :type ESpectatorCameraType%20Enum :default "ThirdPerson"))
  (Rendering|SetVisibility (Variables|Default|GetStaticMesh) (SpectatorPlayerHighlighter|ShouldbeVisible) false))

(event EventFirstPlaceHighlightEnabledChanged
  (params (bInFirstPlaceHighlightEnabled :type Boolean))
  (Rendering|SetVisibility (Variables|Default|GetStaticMesh) (SpectatorPlayerHighlighter|ShouldbeVisible) false))

(event EventBeingFollowedBySpectatorCamChanged
  (params (bInBeingFollowedBySpectatorCam :type Boolean))
  (Rendering|SetVisibility (Variables|Default|GetStaticMesh) (SpectatorPlayerHighlighter|ShouldbeVisible) false))

(event EventPlayerTeamColorChanged
  (params (TeamColor :type Linear%20Color%20Structure%20%28by%20ref%29))
  (Rendering|Material|SetVectorParameterValue (Variables|Default|GetFirstPlacePlayerMaterialInstance) "PlayerTeamColor" TeamColor :!class /Script/Engine.MaterialInstanceDynamic))

(fn ConstructionScript
  (bind _returnvalue (Rendering|Material|CreateDynamicMaterialInstance (Variables|Default|GetStaticMesh) 0 "/Game/Spectating/Effects/FirstPlacePlayerHighlightMaterial.FirstPlacePlayerHighlightMaterial" "None" :!class /Script/Engine.PrimitiveComponent))
  (Variables|Default|SetFirstPlacePlayerMaterialInstance _returnvalue))
